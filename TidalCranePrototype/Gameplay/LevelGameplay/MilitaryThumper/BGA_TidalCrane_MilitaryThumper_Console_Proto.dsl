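; Prototype interactable console that lets a player activate a linked Military Thumper in the TidalCrane prototype.
; It offers an "Activate Thumper" prompt with a timed interaction, gated by a bCanInteract flag.
; On interact it clears that flag, waits through a fixed delay, then calls BeginThumping on the Thumper reference if valid.
; Nothing here re-enables interaction, and BeginPlay, overlap, Tick and the construction script are empty.
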(event EventBeginPlay)

(event Collision|EventActorBeginOverlap
  (params (OtherActor :type Actor%20Object%20Reference)))

(event EventTick
  (params (DeltaSeconds :type Float%20%28single-precision%29)))

(event Interaction|EventBlueprintOnInteract
  (params (InteractingPawn :type Fort%20Pawn%20Object%20Reference) (InteractionBeingAttempted :type EInteractionBeingAttempted%20Enum :default "FirstInteraction"))
  (Variables|Default|SetbCanInteract false)
  (Utilities|FlowControl|Delay 5.0)
  (bind _thumper (Variables|Default|GetThumper)
    (:then
      (Class|BGATidalCraneMilitaryThumperProto|BeginThumping _thumper))
    (:else)))

(fn BlueprintCanInteract
  (params (InteractingPawn :type Fort%20Pawn%20Object%20Reference) (InteractionBeingAttempted :type EInteractionBeingAttempted%20Enum :default "FirstInteraction") (InteractionType :type TInteractionType%20Enum :default "IT_NoInteraction"))
  (return (Variables|Default|GetbCanInteract)))

(fn BlueprintGetInteractionString
  (params (InteractingPawn :type Fort%20Pawn%20Object%20Reference) (InteractionBeingAttempted :type EInteractionBeingAttempted%20Enum :default "FirstInteraction"))
  (return "NSLOCTEXT(\"[A6243E81A461028BC85ADBB5753F2F43]\", \"10DE35EE48767073521DF3BB4271BAC6\", \"Activate Thumper\")"))

(fn BlueprintGetInteractionTime
  (params (InteractingPawn :type Fort%20Pawn%20Object%20Reference) (InteractionBeingAttempted :type EInteractionBeingAttempted%20Enum :default "FirstInteraction"))
  (return 3.0 false))

(fn ConstructionScript)
