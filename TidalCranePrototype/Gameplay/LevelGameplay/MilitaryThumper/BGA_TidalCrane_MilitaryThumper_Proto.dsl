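; Prototype military thumper that pounds a cylinder and triggers a small-creature mound encounter.
; On authority, it enables its child living world volume as a point provider once playlist data is ready, after a short delay.
; BeginThumping plays a slowed timeline driving the cylinder, impact sound and camera shake, then starts the encounter after a delay.
; EndThumping stops the timeline; the bThumping RepNotify toggles audio and resets the cylinder when off.
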(event EventBeginPlay
  (Class|FortAsyncActionCurrentPlaylistReady|PlaylistDataReadyAsync
    (:OnReady
      (Networking|SwitchHasAuthority
        (:Authority
          (Utilities|FlowControl|Delay 2.0)
          (bind _asfort_athena_living_world_volume (Utilities|Casting|CastToFortAthenaLivingWorldVolume (Variables|/Script/Engine.ChildActorComponent|GetChildActor (Variables|Default|GetLivingWorldVolume))))
          (Class|FortAthenaLivingWorldPointProviderInterface|EnablePointProvider_DEPRECATED _asfort_athena_living_world_volume))))))

(event Custom|BeginThumping
  (Networking|SwitchHasAuthority
    (:Authority
      (Variables|Default|SetbThumping true)
      (Utilities|FlowControl|Sequence
        (:then_0
          (Components|Timeline|SetPlayRate (Variables|BGA_TidalCrane_MilitaryThumper_Proto|GetTimeline_Thumping) 0.33)
          (bind (_direction _alpha) (Timeline|Timeline_Thumping Play)
            (:Update
              (Transformation|SetRelativeLocation (Variables|Default|GetCylinder) (Math|Vector|MakeVector 0.0 0.0 (* _alpha 1.5e+02)) false false))
            (:Impact
              (Transformation|SetRelativeLocation (Variables|Default|GetCylinder) "0, 0, 0" false false)
              (Audio|PlaySoundatLocation "/Game/Sounds/Fort_Weapons/Named_Weapons/Boom_Sniper/BoomSniper_Proj_Explode.BoomSniper_Proj_Explode" (Transformation|GetActorLocation) "0, 0, 0" 0.25 0.5)
              (CameraShakes|StartLegacyCameraShakefromSource (Variables|/Script/Engine.PlayerController|GetPlayerCameraManager (Fortnite|GetFirstLocalFortPlayerController true)) "/Game/Effects/Fort_Effects/BluePrints/CameraShakes/B_Medium_Explosion_CameraShake.B_Medium_Explosion_CameraShake_C" (Variables|Default|GetCameraShakeSource) 1.0 "World"))))
        (:then_1
          (Utilities|FlowControl|Delay 5.0)
          (Default|StartActivityEncounter (Game|GetGameState) (Utilities|Casting|CastToFortAthenaLivingWorldVolume (Variables|/Script/Engine.ChildActorComponent|GetChildActor (Variables|Default|GetLivingWorldVolume))) self "/Lager/Lager/Encounters/S37/LagerEncounter_Mound_SmallCreature.LagerEncounter_Mound_SmallCreature"))))))

(event Custom|EndThumping
  (Networking|SwitchHasAuthority
    (:Authority
      (Components|Timeline|Stop (Variables|BGA_TidalCrane_MilitaryThumper_Proto|GetTimeline_Thumping))
      (Variables|Default|SetbThumping false))))

(fn OnRep_bThumping
  (bind _audio (Variables|Default|GetAudio))
  (if (Variables|Default|GetbThumping)
    (Components|Activation|Activate _audio false)
    (else
      (Transformation|SetRelativeLocation (Variables|Default|GetCylinder) "0, 0, 0" false false)
      (Components|Activation|Deactivate _audio))))

(fn ConstructionScript)
