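# EnvQuery that generates candidate points around the querier, named for mound spawning in the MilitaryThumper level gameplay.
# A single donut generator lays points in spiral-patterned rings between an inner and outer radius.
# Points are projected using a navigation filter named for excluding smashable and water areas, then lifted slightly.
# The query has no tests, so candidates are not scored or filtered beyond that projection.
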EnvQuery /TidalCranePrototype/Gameplay/LevelGameplay/MilitaryThumper/EQS_TidalCrane_MilitaryThumper_MoundSpawn
options [{
  generator {
   name=EnvQueryGenerator_Donut_0
   class=EnvQueryGenerator_Donut
   properties {
    bUseSpiralPattern=true
    center=&EnvQueryContext_Querier
    innerRadius=650
    numberOfRings=8
    outerRadius=2000
    pointsPerRing=22
    projectionData {
     extentX=0
     navigationFilter=&NavFilter_Phoebe_NoSmashableNoWater_C
     postProjectionVerticalOffset=15
     projectDown=1500
     projectUp=1500
    }
   }
  }
  tests []
 }]
